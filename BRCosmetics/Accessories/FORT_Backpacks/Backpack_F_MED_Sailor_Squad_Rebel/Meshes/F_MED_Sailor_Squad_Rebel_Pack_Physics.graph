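# Physics asset for the Sailor Squad Rebel backpack cosmetic, giving its creature-like figure and attachments secondary motion.
# Capsule bodies on the dyn_ joints hang from the pack body through two branches, the neck and body master.
# The neck carries ear chains, tongue, head fluff and sword, while the body carries arms, legs, tail, dog tags, zipper and hinged canister.
# Constraints set narrow swing and twist limits, locking selected axes.
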PhysicsAsset /BRCosmetics/Accessories/FORT_Backpacks/Backpack_F_MED_Sailor_Squad_Rebel/Meshes/F_MED_Sailor_Squad_Rebel_Pack_Physics
bodies {
 dyn_L_arm_jnt_1 [{shapeType=Capsule, length=5, radius=2, rotation=<rot 84.8213 0.0005 -7.9036>}]
 dyn_L_arm_jnt_2 [{shapeType=Capsule, length=2.5, radius=1.5, center=<1.8056 0.2516 -0.1616>, rotation=<rot 84.8844 0.0001 -7.9035>}]
 dyn_L_ear_base_jnt [{shapeType=Capsule, length=4, radius=1.5, center=<1.9615 0 -0.1482>, rotation=<rot -85.6696 -180 180>}]
 dyn_L_ear_jnt_1 [{shapeType=Capsule, length=3, radius=1.5, center=<1.2694 0 0.2504>, rotation=<rot 78.8417 180 180>}]
 dyn_L_ear_jnt_2 [{shapeType=Capsule, length=3, radius=1.5, center=<1.2694 0 0.2503>, rotation=<rot 78.8465 179.9999 179.9999>}]
 dyn_L_leg_jnt [{shapeType=Capsule, length=3, radius=2.5, center=<2.5 0.0002 -0.0382>, rotation=<rot 89.1237 -0.0007 0>}]
 dyn_R_arm_jnt_1 [{shapeType=Capsule, length=5, radius=2, center=<3.7676 0.5252 -0.3472>, rotation=<rot 84.7354 0 -7.9044>}]
 dyn_R_arm_jnt_2 [{shapeType=Capsule, length=2.5, radius=1.5, center=<1.8601 0 0.0692>, rotation=<rot 87.8671 180 -179.9999>}]
 dyn_R_ear_base_jnt [{shapeType=Capsule, length=4, radius=1.5, center=<1.9615 0 -0.1482>, rotation=<rot -85.6696 -180 180>}]
 dyn_R_ear_jnt_1 [{shapeType=Capsule, length=3, radius=1.5, center=<1.2694 0 0.2504>, rotation=<rot 78.8411 180 180>}]
 dyn_R_ear_jnt_2 [{shapeType=Capsule, length=3, radius=1.5, center=<1.2694 0 0.2504>, rotation=<rot 78.8408 180 180>}]
 dyn_R_leg_jnt [{shapeType=Capsule, length=3, radius=2.5, center=<2.5 0.0002 -0.0382>, rotation=<rot 89.124 -0.0007 0>}]
 dyn_bodyMaster_jnt [{shapeType=Capsule, length=5, radius=7, center=<-3.8404 -0.5335 1.6625>, rotation=<rot 81.4359 0 0>}]
 dyn_canisterHinge_jnt [{shapeType=Capsule, length=0.5, radius=0.5, center=<0.8595 0 0.2647>, rotation=<rot 22.5278 0 0.0001>}]
 dyn_canister_jnt [{shapeType=Capsule, length=6, radius=2, center=<0.1377 -0.0001 -0.0023>, rotation=<rot -89.0226 179.9998 -179.9998>}]
 dyn_dogTag_jnt_1 [{shapeType=Capsule, length=2, radius=0.5, center=<1.6253 0 0.0938>, rotation=<rot 86.6956 -179.9999 -179.9999>}]
 dyn_dogTag_jnt_2 [{shapeType=Capsule, length=2, radius=0.5, center=<1.6253 0 0.0938>, rotation=<rot 86.6982 179.9999 -179.9998>}]
 dyn_head_fluff_jnt [{shapeType=Capsule, length=2.5, radius=4.5, center=<3.67 0.8729 -0.2983>, rotation=<rot -85.3534 -179.9999 179.9999>}]
 dyn_master_dogTag_jnt [{shapeType=Capsule, length=1, radius=0.5, center=<0.2317 -0.0844 0.1424>, rotation=<rot 0 70 -59.9999>}]
 dyn_neck_joint [{shapeType=Capsule, length=10, radius=3, center=<8.0839 0 -2.021>}]
 dyn_sword_jnt [{shapeType=Capsule, length=15, radius=2.5, rotation=<rot 90 0 0>}]
 dyn_tail_jnt [{shapeType=Capsule, length=4, radius=1, center=<2.0684 0 0.0833>, rotation=<rot 87.6937 179.9998 179.9998>}]
 dyn_tongue_jnt [{shapeType=Capsule, length=2, radius=1.5, center=<1.5 0 0>, rotation=<rot 90 180 180>}]
 dyn_zipper_jnt [{shapeType=Capsule, length=1.8768, radius=0.9135, center=<0.8983 0 0>, rotation=<rot 90 0 0>}]
 pack [{shapeType=Capsule, length=2, radius=2}]
}
constraints [
 {bone1Name=dyn_neck_joint, bone2Name=pack, swing1LimitDegrees=8, swing2LimitDegrees=15, twistLimitDegrees=8}
 {bone1Name=dyn_R_ear_jnt_1, bone2Name=dyn_R_ear_base_jnt, swing1LimitDegrees=30, swing2LimitDegrees=30, twistLimitDegrees=15}
 {bone1Name=dyn_R_ear_jnt_2, bone2Name=dyn_R_ear_jnt_1, swing1LimitDegrees=15, swing2LimitDegrees=15, twistLimitDegrees=15}
 {bone1Name=dyn_L_ear_jnt_1, bone2Name=dyn_L_ear_base_jnt, swing1LimitDegrees=30, swing2LimitDegrees=30, twistLimitDegrees=15}
 {bone1Name=dyn_L_ear_jnt_2, bone2Name=dyn_L_ear_jnt_1, swing1LimitDegrees=15, swing2LimitDegrees=15, twistLimitDegrees=15}
 {bone1Name=dyn_tongue_jnt, bone2Name=dyn_neck_joint, swing1LimitDegrees=15, swing2LimitDegrees=10, twistLimitDegrees=3}
 {bone1Name=dyn_head_fluff_jnt, bone2Name=dyn_neck_joint, swing1LimitDegrees=15, swing2LimitDegrees=15, twistLimitDegrees=5}
 {bone1Name=dyn_bodyMaster_jnt, bone2Name=pack, swing1LimitDegrees=10, swing2LimitDegrees=10, twistLimitDegrees=10}
 {bone1Name=dyn_tail_jnt, bone2Name=dyn_bodyMaster_jnt, swing1LimitDegrees=10, swing2LimitDegrees=30, twistLimitDegrees=10}
 {bone1Name=dyn_dogTag_jnt_1, bone2Name=dyn_master_dogTag_jnt, swing1LimitDegrees=10, swing2LimitDegrees=10, twistLimitDegrees=10}
 {bone1Name=dyn_dogTag_jnt_2, bone2Name=dyn_master_dogTag_jnt, swing1LimitDegrees=10, swing2LimitDegrees=10, twistLimitDegrees=10}
 {bone1Name=dyn_canisterHinge_jnt, bone2Name=dyn_canister_jnt, swing1LimitDegrees=10, swing2LimitDegrees=10, twistLimitDegrees=10}
 {
  bone1Name=dyn_R_arm_jnt_1
  bone2Name=dyn_bodyMaster_jnt
  swing1LimitDegrees=20
  swing2LimitDegrees=40
  twistLimitDegrees=3
  twistMotion=Locked
 }
 {bone1Name=dyn_R_arm_jnt_2, bone2Name=dyn_R_arm_jnt_1, swing1LimitDegrees=10, swing2LimitDegrees=20, twistLimitDegrees=3}
 {
  bone1Name=dyn_L_arm_jnt_1
  bone2Name=dyn_bodyMaster_jnt
  swing1LimitDegrees=20
  swing1Motion=Locked
  swing2LimitDegrees=20
  twistLimitDegrees=3
 }
 {bone1Name=dyn_L_arm_jnt_2, bone2Name=dyn_L_arm_jnt_1, swing1LimitDegrees=20, swing1Motion=Locked, swing2LimitDegrees=20, twistLimitDegrees=3}
 {bone1Name=dyn_master_dogTag_jnt, bone2Name=dyn_bodyMaster_jnt, swing1LimitDegrees=5, swing2LimitDegrees=5, twistLimitDegrees=5}
 {
  bone1Name=dyn_zipper_jnt
  bone2Name=dyn_bodyMaster_jnt
  swing1LimitDegrees=20
  swing2LimitDegrees=45
  swing2Motion=Locked
  twistLimitDegrees=5
 }
 {bone1Name=dyn_canister_jnt, bone2Name=dyn_bodyMaster_jnt, swing1LimitDegrees=7, swing2LimitDegrees=8, twistLimitDegrees=5}
 {
  bone1Name=dyn_sword_jnt
  bone2Name=dyn_neck_joint
  swing1LimitDegrees=3
  swing1Motion=Locked
  swing2LimitDegrees=4
  twistLimitDegrees=45
  twistMotion=Locked
 }
 {
  bone1Name=dyn_R_leg_jnt
  bone2Name=dyn_bodyMaster_jnt
  swing1LimitDegrees=30
  swing2LimitDegrees=30
  twistLimitDegrees=15
  twistMotion=Locked
 }
 {
  bone1Name=dyn_L_leg_jnt
  bone2Name=dyn_bodyMaster_jnt
  swing1LimitDegrees=30
  swing2LimitDegrees=30
  twistLimitDegrees=15
  twistMotion=Locked
 }
 {
  bone1Name=dyn_L_ear_base_jnt
  bone2Name=dyn_neck_joint
  swing1LimitDegrees=20
  swing2LimitDegrees=10
  twistLimitDegrees=45
  twistMotion=Locked
 }
 {
  bone1Name=dyn_R_ear_base_jnt
  bone2Name=dyn_neck_joint
  swing1LimitDegrees=10
  swing2LimitDegrees=15
  twistLimitDegrees=45
  twistMotion=Locked
 }
]
