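# Animation Blueprint for the Sailor Squad Rebel backpack, adding procedural motion and toggling a material element.
# The AnimGraph rotates master and tail bones while skydiving, then runs RigidBody physics tuned for front-end and crouch states.
# Weapon, customization and skydive events set "Hide Element 01" on the dynamic material, based on whether a SailorSquadRebel-tagged pickaxe is equipped.
# Skydive start clears the parameter; the weapon-change sequence's second branch overrides its skydiving check.
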AnimBlueprint /BRCosmetics/Accessories/FORT_Backpacks/Backpack_F_MED_Sailor_Squad_Rebel/Meshes/F_MED_Sailor_Squad_Rebel_Pack_AnimBP
target_skeleton=&/F_MED_Sailor_Squad_Rebel_Pack_Skeleton
graphs [
 AnimGraph {nodes [
   n1 = Local To Component(LocalPose=Input Pose())
   n2 = Get SkydiveRotate() -> (SkydiveRotate:Rotator)
   n3 = Get bIsInFrontEnd()
   n4 = OR Boolean(A=Get bIsCrouching(), B=Get bIsCrouchSprinting())
   n5 = RigidBody!(
    ComponentPose="Transform (Modify) Bone - Bone: dyn_tail_jnt"(
     ComponentPose="Transform (Modify) Bone - Bone: dyn_master_rotate_master_jnt"(ComponentPose=@n1, Alpha=1, Translation="0,0,0", Rotation=@n2, Scale="1,1,1") {rotationMode=BMM_Additive, alphaBoolBlend {blendOption=Linear}, boneToModify {boneName=dyn_master_rotate_master_jnt}}
     Alpha=1
     Translation="0,0,0"
     Rotation=Get SkydiveTailRotate() -> (SkydiveTailRotate:Rotator)
     Scale="1,1,1"
    ) {rotationMode=BMM_Replace, alphaBoolBlend {blendOption=Linear}, boneToModify {boneName=dyn_tail_jnt}}
    Alpha=SelectFloat(A=0.3, B=1, bPickA=@n3)
    OverrideWorldGravity:Vector=SelectVector(A="0,0,-300", B="0,0,-980", bPickA=@n3)
    ExternalForce
    ComponentLinearAccScale:Vector=SelectVector(A="0.01,0.01,0.01", B="0.23,0.3,0.23", bPickA=@n4)
    ComponentLinearVelScale:Vector=SelectVector(A="0.6,0.6,0.6", B="0.33,0.33,0.33", bPickA=@n4)
   ) -> (Pose) {
    bClampLinearTranslationLimitToRefPose=true
    bOverrideWorldGravity=true
    overridePhysicsAsset=&/F_MED_Sailor_Squad_Rebel_Pack_Physics
    alphaBoolBlend {blendOption=Linear}
    baseBoneRef {boneName=root}
    componentAppliedLinearAccClamp=<6000 6000 6000>
    componentLinearAccScale=<0.23 0.23 0.23>
    componentLinearVelScale=<0.33 0.33 0.33>
    overrideWorldGravity {z=-980}
   }
   Output Pose(Result=Component To Local(ComponentPose=@n5)) {layerGroup=None}
  ]}
 EventGraph {nodes [
   Event BlueprintUpdateAnimation()
   > n1 = Sequence()
   > Set SkydiveRotate(SkydiveRotate:Rotator=SelectRotator!(A:Rotator="0,0,20", B:Rotator="0, 0, 0", bPickA:bool=Get bIsSkyDiving()) -> (ReturnValue:Rotator)) -> (Output_Get:Rotator)
   Set SkydiveTailRotate(execute=@n1.then_1, SkydiveTailRotate:Rotator=SelectRotator(A="0,0,50", B="0, 0, 0", bPickA=Get bIsSkyDiving())) -> (Output_Get:Rotator)
   n2 = Event OnWeaponChanged()
   n3 = Get Dynamic Material Instance(execute=[@n2.then, @n11.then], SkeletalMeshComponent=GetPartSkeletalMeshComponent())
   > n4 = Sequence()
   n5 = PickaxeItemDefinitionHasAnyMetaTag(Tags=Make Array("[0]"=(TagName="Cosmetics.MetaTags.SailorSquadRebel")), bRequiresEquipped=true)
   > n6 = Branch(Condition=@n5)
   > n7 = Branch(Condition=Get bIsSkyDiving())
   SetScalarParameterValue(execute=[@n9.then, @n7.then, @n8.else, @n6.else], self=@n3, ParameterName=Hide Element 01, Value=0)
   SetScalarParameterValue(execute=[@n7.else, @n8.then], self=@n3, ParameterName=Hide Element 01, Value=1)
   n8 = Branch(execute=[@n10.then, @n4.then_1], Condition=@n5)
   n9 = Event OnBeginSkydiving()
   n10 = Event OnEndSkydiving()
   n11 = Event OnCharacterCustomizationLoaded()
  ]}
]
variables [
 SkydiveRotate {type=Rotator}
 SkydiveTailRotate {type=Rotator}
]
